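; Grapple hook projectile for the hook gun that pulls the firing player, or their vehicle, toward the impact point.
; BeginPlay spawns the rope on the server; EventHit tightens it and attaches the hook to pawns or certain moving objects.
; Server-triggered Travel consumes ammo and launches along suggested arcs, factoring in momentum, with effects and fall-damage immunity.
; Launchpad or skydive launches cancel the hook, weapon switches after grappling destroy it, and EventDestroyed handles rope cleanup.
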(event Custom|LaunchCheck
  ;[1 1 1] Reset everything if a launchpad is touched while Proj is active
  (bind _jumppadding (CallFunction|HandleLaunch))
  (if _jumppadding
    (Utilities|FlowControl|DoOnce false
      (:Completed
        (Variables|Default|SetLaunchPad true)
        (Variables|Default|SetHookStopped true)
        (Ability|SendGameplayEventtoActor (Variables|Default|GetPlayerPawn) "(TagName=\"Abilities.Grapple.Hit\")" (Utilities|Struct|MakeGameplayEventData :Instigator self :ContextHandle "()" :InstigatorTags "(GameplayTags=)" :TargetTags "(GameplayTags=)" :EventMagnitude 0.0 :TargetData "()"))
        (bind _rope (Variables|Default|GetRope)
          (:then
            (Actor|DestroyActor _rope)
            (Utilities|FlowControl|Delay 0.1) ; so proj and rope can't be destroyed at the same time
            (Actor|DestroyActor))
          (:else))))))

(event Custom|Grapple
  (bind _playerpawn (Variables|Default|GetPlayerPawn))
  (bind _playerpawn_2 (Variables|Default|GetPlayerPawn))
  (bind _charactermovement (Variables|/Script/Engine.Character|GetCharacterMovement _playerpawn_2))
  (bind _playerpawn_3 (Variables|Default|GetPlayerPawn))
  (bind _playerpawn_4 (Variables|Default|GetPlayerPawn))
  ;[1 1 1] Grapple - Calculate the launch values, blocking and clearing movement as needed. Plays Visuals for lense effects and shakes.
  (Pawn|Components|CharacterMovement|ClearAccumulatedForces _charactermovement)
  (Components|Movement|StopMovementImmediately _charactermovement)
  (if (not (Variables|Default|GetAirControl))
    (Input|SetIgnoreMoveInput (Game|GetInstigatorController (Variables|Default|GetPlayerPawn)) true))
  (Audio|SpawnSoundAttached (Variables|GrapplerSounds|GetTravelSound) (Variables|/Script/Engine.Actor|GetRootComponent (Variables|Default|GetPlayerPawn)))
  (HandleSticks|HandleMomentum)
  (FortMovement|CancelJumpAbility _playerpawn false)
  (if (Math|Boolean|ANDBoolean (AI|Components|NavMovement|IsFalling (Variables|/Script/Engine.Character|GetCharacterMovement _playerpawn)) (Math|Float|InRange(Float) (Variables|Default|GetFireAImPitch) (Variables|Default|GetDownPitchMin) (Variables|Default|GetDownPitchMax)) (Variables|Default|GetDoesCrouchCheck))
    (Variables|Default|SetPreLaunchVelocity (Variables|/Script/Engine.MovementComponent|GetVelocity (Pawn|GetMovementComponent (Variables|Default|GetPlayerPawn))))
    (bind _outlaunchvelocity (Game|SuggestProjectileVelocityCustomArc (Transformation|GetActorLocation (Variables|Default|GetPlayerPawn)) (Variables|Default|GetHitLocation) (CallFunction|GetOverrideGravity 0.0) 0.9))
    (if (Utilities|Enum|NotEqual(Enum) (Networking|GetLocalRole (Variables|Default|GetPlayerPawn)) "ROLE_SimulatedProxy")
      (Character|LaunchCharacter _playerpawn_4 (+ _outlaunchvelocity (Variables|Default|GetPreLaunchVelocity)) true false)
      (Feedback|SpawnGenericCameraLensEffect _playerpawn_4 (LensEffect|GetInterfaceClass (Variables|Default|GetLensEffectInterfaceClass))))
    (else
      (Character|UnCrouch (Variables|Default|GetPlayerPawn))
      (bind _outlaunchvelocity_2 (Game|SuggestProjectileVelocityCustomArc (Transformation|GetActorLocation (Variables|Default|GetPlayerPawn)) (Math|Vector|MakeVector (.x (Variables|Default|GetHitLocation)) (.y (Variables|Default|GetHitLocation)) (.z (Variables|Default|GetHitLocation))) (CallFunction|GetOverrideGravity (Variables|Default|GetArcGravityOverride)) (ScalableFloat|GetValueatLevel (Variables|Default|GetRow_PlayerArc)))) ; Calculates how much force is needed to move actor to target location
      (if (Utilities|Enum|NotEqual(Enum) (Networking|GetLocalRole (Variables|Default|GetPlayerPawn)) "ROLE_SimulatedProxy")
        (bind _output_get (Variables|Default|SetLaunchVelocity (select (Variables|Default|GetHasMinUpwardLaunchVelocity) (Math|Vector|MakeVector (.x (+ _outlaunchvelocity_2 (CallFunction|MomentumCheck _outlaunchvelocity_2 _outlaunchvelocity_2))) (.y (+ _outlaunchvelocity_2 (CallFunction|MomentumCheck _outlaunchvelocity_2 _outlaunchvelocity_2))) (.z (+ _outlaunchvelocity_2 (CallFunction|MomentumCheck _outlaunchvelocity_2 _outlaunchvelocity_2)))) (+ _outlaunchvelocity_2 (CallFunction|MomentumCheck _outlaunchvelocity_2 _outlaunchvelocity_2)))))
        (Character|LaunchCharacter _playerpawn_3 _output_get true true) ; Add impulse isn't resetting fall damage when changing momentum 
        (Feedback|SpawnGenericCameraLensEffect _playerpawn_3 (LensEffect|GetInterfaceClass (Variables|Default|GetLensEffectInterfaceClass)))
        (Game|Feedback|ClientStartCameraShake :self (Fortnite|GetFortPlayerControllerfromActor _playerpawn_3) :Shake (Variables|Default|GetCamShake_Grapple) :Scale 1.0 :PlaySpace "CameraLocal"))))
  (CallFunction|ApplyTillLandFallDamageImmunity) ; Look to apply Fall Damage Immunity only after launching
)

(event Custom|WeaponCheck
  ;[1 1 1] Destroys projectil if weapon is switched.  This is so we can keep the rope FX looking good, but still let people weapon switch after grappling
  (if (not (Utilities|Equal(Object) (Variables|/Script/FortniteGame.FortPawn|GetCurrentWeapon (Variables|Default|GetPlayerPawn)) (Variables|Default|GetWeapon%20Actor)))
    (Actor|DestroyActor)))

(event EventBeginPlay
  (bind _playerpawn (Variables|Default|GetPlayerPawn))
  (bind _returnvalue (Fortnite|GetFortPlayerControllerfromActor _playerpawn))
  (bind _asfort_player_controller_athena (Utilities|Casting|CastToFortPlayerControllerAthena _returnvalue))
  (bind _capsulecomponent (Variables|Collision|GetCapsuleComponent))
  (bind _playerpawn_2 (Variables|Default|GetPlayerPawn))
  ;[1 1 1] Begin Play
  (bind _asfort_player_pawn (Utilities|Casting|CastToFortPlayerPawn (Variables|Actor|GetInstigator)))
  (Variables|Default|SetPlayerPawn _asfort_player_pawn)
  (Tags|AddNonReplicatedTagtoPawn (Variables|Default|GetPlayerPawn) (Variables|Default|GetRopeActiveTag))
  (Utilities|IsValid _asfort_player_controller_athena
    (:Is%20Valid
      (EventDispatchers|BindEventtoOnBeginSkydiving _asfort_player_controller_athena (EventDispatchers|CreateEvent :!event Custom|LaunchCheck))
      (bind _asb_hookgun_athena (Utilities|Casting|CastToB_Hookgun_Athena (Actor|GetOwner)))
      (Variables|Default|SetWeapon%20Actor _asb_hookgun_athena)
      (bind _playerpawn_3 (Variables|Default|GetPlayerPawn)
        (:then
          (if (Networking|IsServer)
            (CallFunction|SpawnRope _playerpawn_3)
            (bind _returnvalue_2 (GameplayEffects|ApplyGameplayEffectToTarget (Ability|GetAbilitySystemComponent _playerpawn_3) (Variables|Default|GetGE_RopeActive) (Ability|GetAbilitySystemComponent _playerpawn_3) 1.0 (Utilities|Struct|MakeGameplayEffectContextHandle)))
            (bind _rope (Variables|Default|GetRope)
              (:then
                (Variables|/Game/Weapons/FORT_Crossbows/Blueprints/B_HookGunRope.B_HookGunRope_C|SetRopeGE _returnvalue_2 _rope))
              (:else))
            (else
              _))
          (Variables|Default|SetFireAImPitch (FortPlayerPawn|GetBaseAimPitchQuantized (Variables|Default|GetPlayerPawn)))
          (Tags|AddNonReplicatedTagtoPawn (Variables|Default|GetPlayerPawn) (Variables|Default|GetBlockIncomingPickups))
          (if (Vehicle|IsinVehicle _playerpawn_2)
            (bind _returnvalue_3 (Vehicle|GetVehicleActor (Vehicle|GetVehicleInterface _playerpawn_2) :!class /Script/FortniteGame.FortVehicleInterface))
            (Collision|IgnoreActorwhenMoving _capsulecomponent _returnvalue_3 true)
            (bind _returnvalue_4 (Vehicle|GetAllPassengers(Message) _returnvalue_3))
            (for _array_element _returnvalue_4
              (Collision|IgnoreActorwhenMoving _capsulecomponent _array_element true))))
        (:else)))))

(event Collision|EventHit
  (params (MyComp :type Primitive%20Component%20Object%20Reference) (Other :type Actor%20Object%20Reference) (OtherComp :type Primitive%20Component%20Object%20Reference) (bSelfMoved :type Boolean) (HitLocation :type Vector :default "0, 0, 0") (HitNormal :type Vector :default "0, 0, 0") (NormalImpulse :type Vector :default "0, 0, 0") (Hit :type Hit%20Result%20Structure%20%28by%20ref%29))
  (bind (_bblockinghit _binitialoverlap _time _distance _location _impactpoint _normal _impactnormal _physmat _hitactor) (Collision|BreakHitResult Hit))
  ;[1 1 1] Stop event.  Sets variables and calls the travel event
  (if (Utilities|NotEqual(Object) self _hitactor)
    (bind _playerpawn (Variables|Default|GetPlayerPawn)
      (:then
        (Tether|SetTetherPawn _playerpawn 0)
        (if (Variables|/Script/FortniteGame.FortPawn|GetbIsDBNO _playerpawn)
          (bind _playerpawn_2 (Variables|Default|GetPlayerPawn)
            (:then
              (Ability|SendGameplayEventtoActor _playerpawn_2 (Variables|Default|GetEvent_Returned)) ; Unblock pawn from weapon switching when interp finishes
              (bind _rope (Variables|Default|GetRope)
                (:then
                  (Rendering|SetActorHiddenInGame _rope true)
                  (Actor|DestroyActor _rope))
                (:else)))
            (:else))
          (else
            (CallFunction|HandleHits Hit)
            (Actor|SetLifeSpan (Variables|Default|GetLifespan%20on%20Hit))
            (Utilities|ShouldPulltoPawnWhileinVehicle? Hit
              (:True
                (if (and (Networking|IsServer) (Utilities|IsValid (Game|GetInstigatorController) :!class /Script/Engine.KismetSystemLibrary))
                  (CallFunction|Travel (Variables|Default|GetHitLocation))))))))
      (:else))))

(event Custom|Travel
  (params (HitLocation :type Vector))
  (bind _playerpawn (Variables|Default|GetPlayerPawn))
  (bind _returnvalue (Vehicle|GetVehicleInterface _playerpawn))
  (bind _returnvalue_2 (Utilities|ToObject(Interface) _returnvalue))
  (bind _playerpawn_2 (Variables|Default|GetPlayerPawn))
  (bind _returnvalue_3 (Pawn|IsLocallyControlled _playerpawn_2))
  (bind (_asfort_athena_jackal_vehicle _bsuccess) (Utilities|Casting|CastToFortAthenaJackalVehicle _returnvalue_2))
  (bind _sound_stick2d (Variables|GrapplerSounds|GetSound_Stick2D))
  (bind _sound_stick (Variables|GrapplerSounds|GetSound_Stick))
  (bind _returnvalue_4 (select _returnvalue_3 _sound_stick2d _sound_stick))
  (bind _mesh (Variables|/Script/Engine.Character|GetMesh _playerpawn))
  (bind _weapon_actor (Variables|Default|GetWeapon%20Actor))
  (bind _playerpawn_3 (Variables|Default|GetPlayerPawn))
  (bind _weapon_actor_2 (Variables|Default|GetWeapon%20Actor))
  ;[1 1 1] Travel event.  Plays travel montage, sounds and then moves the player.
  (Variables|Default|SetHitLocation HitLocation)
  (Networking|SwitchHasAuthority
    (:Authority
      (Utilities|IsValid _weapon_actor
        (:Is%20Valid
          (if (and (Variables|Default|GetConsumesAmmo) (not (FortWeapon|Ammo|HasInfiniteUses _weapon_actor)))
            (FortWeapon|Ammo|ConsumeAmmo _weapon_actor_2 1 false)
            (else
              (FortWeapon|Ammo|ConsumeAmmo _weapon_actor_2 0 false))))
        (:Is%20Not%20Valid
          _))
      (bind _jumppadding (CallFunction|HandleLaunch))
      (if _jumppadding
        (CallFunction|LaunchCheck) ; client/server run into disagreements here - this is a failsafe
        (else
          (bind _returnvalue_5 (Zipline|IsZiplining _playerpawn_3))
          (if _returnvalue_5
            (Zipline|EndZiplining _playerpawn_3 false false)
            (else
              _))
          (bind _rootcomponent (Variables|/Script/Engine.Actor|GetRootComponent (Variables|Default|GetPlayerPawn))
            (:then
              (Audio|SpawnSoundAttached :Sound _returnvalue_4 :AttachToComponent _rootcomponent :AttachPointName "None" :LocationType "KeepRelativeOffset" :bStopWhenAttachedToDestroyed true :VolumeMultiplier 1.0 :PitchMultiplier 1.0 :StartTime 0.0 :bAutoDestroy true))
            (:else
              (Audio|PlaySoundatLocation _returnvalue_4 (Variables|Default|GetHitLocation))))
          (CallFunction|JackalCheck) ; This is to fix an animation bug
          (Ability|SendGameplayEventtoActor (Variables|Default|GetPlayerPawn) (Variables|Default|GetEvent_GrappleHit) (Utilities|Struct|MakeGameplayEventData :Instigator self :ContextHandle "()" :InstigatorTags "(GameplayTags=)" :TargetTags "(GameplayTags=)" :EventMagnitude 0.0 :TargetData "()"))
          (Utilities|IsValid (Fortnite|GetFortPlayerControllerfromActor _playerpawn)
            (:Is%20Valid
              _)
            (:Is%20Not%20Valid
              (if _bsuccess
                (Animation|Montage|PlayMontage _mesh (Variables|Default|GetAnim_GrappleHoverboard)) ; These two montage nodes are hacks to make grapple anims work correctly for replays
                (else
                  (Animation|Montage|PlayMontage _mesh (Variables|Default|GetAnim_Grapple) 1.0 0.0 (CallFunction|PickGrappleMontage (Variables|Default|GetFireAImPitch) _bsuccess))))))
          (bind _playerpawn_4 (Variables|Default|GetPlayerPawn)
            (:then
              (if (Vehicle|IsinVehicle _playerpawn_4)
                (CallFunction|GrappleVehicle)
                (else
                  (CallFunction|Grapple)))
              (Utilities|Time|SetTimerbyFunctionName 0 "WeaponCheck" 0.1 true) ; destroy projectile if weapons switched
              (CallFunction|GiveQuestCredit))
            (:else)))))
    (:Remote
      _)))

(event OnEnterWaterEvent(WaterInteractionComponent)
  (params (WaterBody :type Fort%20Water%20Body%20Actor%20Object%20Reference) (WaterInteractionComponent :type Fort%20Water%20Interaction%20Component%20Object%20Reference) (bIsFirstBody :type Boolean))
  (bind _waterinteractioncomponent (Variables|Water|GetWaterInteractionComponent))
  (bind (_waterplanelocation _waterplanenormal _watersurfaceposition _waterdepth _waterbodyidx) (FX|GetLastWaterSurfaceInfo _waterinteractioncomponent :!class /Script/FortniteGame.FortWaterInteractionComponent))
  ;[1 1 1] Water splash
  (if bIsFirstBody
    (bind _returnvalue (Niagara|SpawnSystematLocation (Variables|Default|GetFX_WaterImpact) _watersurfaceposition (Math|Rotator|MakeRotfromX _waterplanenormal)))
    (Effects|Components|ParticleSystem|SetFloatParameter _returnvalue "WaterDepth" _waterdepth)
    (Effects|Components|ParticleSystem|SetFloatParameter _returnvalue "WaterBodyIndex" (Math|Conversions|ToFloat(Integer) _waterbodyidx))
    (Audio|SpawnSoundatLocation (Variables|GrapplerSounds|GetSound_Water) _waterplanelocation)))

(event EventDestroyed
  ;[1 1 1] Destroyed
  (bind _playerpawn (Variables|Default|GetPlayerPawn)
    (:then
      (if (Variables|/Script/FortniteGame.FortPawn|GetbIsDBNO _playerpawn)
        (Ability|SendGameplayEventtoActor (Variables|Default|GetPlayerPawn) (Variables|Default|GetEvent_Returned)) ; Unblock pawn from weapon switching when interp finishes
        (bind _rope (Variables|Default|GetRope)
          (:then
            (Actor|DestroyActor _rope))
          (:else))
        (elif (Variables|Default|GetHookStopped)
          (Niagara|SpawnSystematLocation (Variables|Default|GetFX_Reattach) (Transformation|GetActorLocation) (Math|Conversions|RotationFromXVector (Variables|Default|GetImpactNormal)))
          (Audio|PlaySoundatLocation (Variables|GrapplerSounds|GetSound_RopeDestroy) (Transformation|GetActorLocation (Variables|Default|GetPlayerPawn)))
          (bind _rope_2 (Variables|Default|GetRope)
            (:then
              (Actor|DestroyActor _rope_2))
            (:else))
          (else
            (bind _playerpawn_2 (Variables|Default|GetPlayerPawn)
              (:then
                (Animation|Montage|PlayMontage (Variables|/Script/Engine.Character|GetMesh _playerpawn_2) (Variables|Default|GetReattachMontage) 1.2)
                (bind _rootcomponent (Variables|/Script/Engine.Actor|GetRootComponent _playerpawn_2)
                  (:then
                    (Audio|SpawnSoundAttached (Variables|GrapplerSounds|GetSound_YankBack) _rootcomponent))
                  (:else
                    _))
                (bind _rope_3 (Variables|Default|GetRope)
                  (:then
                    (Variables|/Script/FortniteGame.FortHookGunRope|SetbProjectileGone true _rope_3))
                  (:else)))
              (:else))))))
    (:else)))

(event Custom|GrappleVehicle
  (bind _playerpawn (Variables|Default|GetPlayerPawn))
  (bind _returnvalue (Vehicle|GetVehicleInterface _playerpawn))
  (bind _returnvalue_2 (Class|FortVehicleInterface|GetOwnerLocation _returnvalue))
  (bind _vehiclevelocity (Variables|Default|GetVehicleVelocity))
  (bind _momentum (Variables|Default|GetMomentum))
  ;[1 1 1] Vehicle Grappling
  (if (not (GameplayTags|HasAnyTags (Class|FortVehicleInterface|GetGameplayTags _returnvalue) (Variables|Default|GetTC_VehicleIgnorePull) true))
    (bind _returnvalue_3 (Vehicle|GetVehicleActor _returnvalue :!class /Script/FortniteGame.FortVehicleInterface))
    (bind _asfort_player_pawn (Utilities|Casting|CastToFortPlayerPawn (Variables|Default|GetStruckActor))
      (:then
        (bind _returnvalue_4 (Vehicle|GetAllPassengers(Message) _returnvalue_3)))
      (:CastFailed
        _))
    (if (not (Utilities|Array|ContainsItem _returnvalue_4 _asfort_player_pawn))
      (HandleSticks|HandleMomentum)
      (Variables|Default|SetVehicleVelocity (Vehicle|GetVehicleVelocity _returnvalue))
      (Utilities|Casting|CastToFortAthenaJackalVehicle _returnvalue_3
        (:then
          (Variables|VehicleGrappling|SetbInHoverboard true))
        (:CastFailed
          (Variables|VehicleGrappling|SetbInHoverboard false)))
      (bind _outlaunchvelocity (Game|SuggestProjectileVelocityCustomArc _returnvalue_2 (Variables|Default|GetHitLocation) 0.0 (CallFunction|GetVehicleArc))) ; Calculates how much force is needed to move actor to target location
      (if (> (Math|Vector|VectorLength (+ (+ _outlaunchvelocity _momentum) _vehiclevelocity)) (Math|Vector|VectorLength (+ _outlaunchvelocity _momentum)))
        (bind _outlaunchvelocity_2 (Game|SuggestProjectileVelocityCustomArc _returnvalue_2 (Variables|Default|GetHitLocation) 0.0 (CallFunction|GetVehicleArc))) ; 40% of the distance + 40% of your veloctiy
        (else
          (bind _outlaunchvelocity_3 (Game|SuggestProjectileVelocityCustomArc _returnvalue_2 (Variables|Default|GetHitLocation) 0.0 (CallFunction|GetVehicleArc))) ; Go 40% of the distance
        ))
      (Audio|BroadcastAbilityImpactatLocation "/Game/Athena/Sounds/Events/Cube/Cube_LaunchPlayer_Cue.Cube_LaunchPlayer_Cue" _returnvalue_3 (* (select (> (Math|Vector|VectorLength (+ (+ _outlaunchvelocity _momentum) _vehiclevelocity)) (Math|Vector|VectorLength (+ _outlaunchvelocity _momentum))) (+ _outlaunchvelocity_2 (Variables|Default|GetVehicleVelocity)) _outlaunchvelocity_3) (CallFunction|GetVehicleLaunchMult)) 7.8e+02) ; This is the node that launches vehicles
    )))

(fn ApplyTillLandFallDamageImmunity
  (Networking|CanExecuteCosmeticEvents
    (:False
      (bind _playerpawn (Variables|Default|GetPlayerPawn)
        (:then
          ;[0.4582 0.9901 1] Check for Creative/UEFN island setting
          (if (Fortnite|IsSpatialGameplayExperience)
            (if (ScalableFloat|GetValueasBoolwithContextParam _playerpawn (Variables|Default|GetIslandSetting_GrappleFallDamageImmunity))
              _)
            (else
              _))
          ;[1 1 1] Apply Till Land Fall Damage Immunity, but only if the player doesn't have the matching gameplay tag already applied
          (if (not (GameplayTags|HasAnyMatchingGameplayTags _playerpawn (Variables|Default|GetTC_IgnoreFallDamageApplication)))
            (Environment|ApplyGlobalEnvironmentGameplayEffecttoActor _playerpawn (Variables|Default|GetGE_TillLandFallDamageImmunity_Class) 1)))
        (:else)))))

(fn CalculateYankForce
  (bind (_outlocation _outrotation) (Actor|GetActorEyesViewPoint (Variables|Default|GetPlayerPawn)))
  (bind _outlaunchvelocity (Game|SuggestProjectileVelocityCustomArc (Transformation|GetActorLocation (Variables|Default|GetTarget)) (+ _outlocation (* (Math|Vector|GetForwardVector _outrotation) (Variables|Default|GetDistanceFromInstigator))))) ; Finds out how much force is needed to put the target 300 units in front of where instigator is facing
  (if (Utilities|Enum|NotEqual(Enum) (Networking|GetLocalRole (Variables|Default|GetTarget)) "ROLE_SimulatedProxy")
    (Character|LaunchCharacter (Variables|Default|GetTarget) (+ _outlaunchvelocity (Math|Vector|MakeVector (.x (Transformation|GetVelocity (Variables|Default|GetPlayerPawn))) (.y (Transformation|GetVelocity (Variables|Default|GetPlayerPawn))) (Math|Float|Absolute(Float) (.z (Transformation|GetVelocity (Variables|Default|GetPlayerPawn)))))) true true)))

(fn GetOverrideGravity
  (params (OverrideGravityValue :type Float%20%28double-precision%29))
  (return (select (Variables|Default|GetShouldCompenstateForGravity) (Components|Movement|GetGravityZ (Variables|/Script/Engine.Character|GetCharacterMovement (Variables|Default|GetPlayerPawn))) OverrideGravityValue)))

(fn GetVehicleArc
  ;[1 1 1] Grab Arc Values for vehicles from rows
  (return (select (Variables|VehicleGrappling|GetbInHoverboard) (ScalableFloat|GetValueatLevel (Variables|VehicleGrappling|GetRow_JackleArc)) (ScalableFloat|GetValueatLevel (Variables|VehicleGrappling|GetRow_VehicleArc)))))

(fn GetVehicleLaunchMult
  ;[1 1 1] Get Vehicle Force Multiplier
  (return (select (Variables|VehicleGrappling|GetbInHoverboard) (ScalableFloat|GetValueatLevel (Variables|VehicleGrappling|GetRow_JackleForceMultiplier)) (ScalableFloat|GetValueatLevel (Variables|VehicleGrappling|GetRow_VehicleForceMultiplier)))))

(fn GiveQuestCredit
  ;[1 1 1] Give Quest credit
  (if (Networking|IsServer)
    (Quest|SendComplexCustomStatEvent (Quests|GetQuestManager (Fortnite|GetFortPlayerControllerfromActor (Variables|Actor|GetInstigator)) "Invalid") 0 (Variables|Default|GetT_QuestTag) "(GameplayTags=)" 1)))

(fn HandleHits
  (params (Hit :type Hit%20Result%20Structure))
  (bind (_bblockinghit _binitialoverlap _time _distance _location _impactpoint _normal _impactnormal _physmat _hitactor _hitcomponent) (Collision|BreakHitResult Hit))
  (Variables|Default|SetHookStopped true)
  (Variables|Default|SetHitLocation _impactpoint)
  (bind _rope (Variables|Default|GetRope)
    (:then
      (Class|BHookGunRope|TightenRope _rope)
      (Variables|Default|SetImpactNormal _impactnormal)
      (Transformation|SetActorRotation (Math|Conversions|RotationFromXVector (* _impactnormal -1.0)))
      (Game|Feedback|ClientPlayForceFeedback (Fortnite|GetFortPlayerControllerfromActor (Variables|Default|GetPlayerPawn)) "/Game/Weapons/ForceFeedback/FF_Generic_Throw.FF_Generic_Throw" "None" false false false)
      (bind _asfort_pawn (Utilities|Casting|CastToFortPawn _hitactor)
        (:then
          (Gameplay|ServerAttachtoActor Hit _asfort_pawn (Variables|Default|GetOffset%20Distance%20from%20Physics%20Mesh) (Variables|Default|GetOffsetDistanceFromBone) 0.75))
        (:CastFailed
          (HandleSticks|HandleSticks _hitactor _hitcomponent)))
      (Variables|Default|SetStruckActor _hitactor)
      (return 0))
    (:else)))

(fn HandleLaunch
  (if (Math|Boolean|ANDBoolean (not (Variables|/Script/FortniteGame.FortPlayerPawn|GetbIsSkydiving (Variables|Default|GetPlayerPawn))) (not (Parachute|GetPendingSkydiveLaunch (Variables|Default|GetPlayerPawn))) (not (Parachute|IsSkydiving (Variables|Default|GetPlayerPawn))))
    (return false)
    (else
      (return true))))

(fn HandleMomentum
  (bind _struckactor (Variables|Default|GetStruckActor))
  (Utilities|IsValid _struckactor
    (:Is%20Valid
      (bind _asfort_player_pawn (Utilities|Casting|CastToFortPlayerPawn _struckactor)
        (:then
          (if (and (Vehicle|IsinVehicle _asfort_player_pawn) false)
            (Variables|Default|SetMomentum (Vehicle|GetVehicleVelocity (Vehicle|GetVehicleInterface _asfort_player_pawn)))
            (else
              _)))
        (:CastFailed
          _))
      (Variables|Default|SetMomentum (Transformation|GetVelocity (Variables|Default|GetStruckActor))))))

(fn HandleSticks
  (params (Object :type Object%20Reference) (Component :type Primitive%20Component%20Object%20Reference))
  (Utilities|FlowControl|Sequence
    (:then_0
      (HandleSticks|HandleStuckBuilding Object Component))
    (:then_1
      (HandleSticks|HandleStuckVehicle Object Component))
    (:then_2
      (HandleSticks|HandleStuckSupplyDrop Object Component))
    (:then_3
      (HandleSticks|HandleStuckPhysicsObject Object Component))
    (:then_4
      (HandleSticks|HandleStuckMovingPlatform Object Component))))

(fn HandleStuckBuilding
  (params (Object :type Object%20Reference) (Component :type Primitive%20Component%20Object%20Reference))
  (bind _asbuilding_wall (Utilities|Casting|CastToBuildingWall Object))
  (LocalVariables|SetLocalWall _asbuilding_wall)
  (bind _asstatic_mesh_component (Utilities|Casting|CastToStaticMeshComponent Component))
  (bind _returnvalue (BuildingWall|IsDoorComponent (LocalVariables|GetLocalWall) _asstatic_mesh_component))
  (if _returnvalue
    (Transformation|AttachActorToComponent Component "None" "KeepWorld" "KeepWorld" "KeepWorld" false)))

(fn HandleStuckMovingPlatform
  (params (Object :type Object%20Reference) (Component :type Primitive%20Component%20Object%20Reference))
  (bind _asactor (Utilities|Casting|CastToActor Object))
  (Utilities|IsValid (Actor|GetComponentbyClass _asactor "/Script/FortniteGame.FortMovingPlatformComponent")
    (:Is%20Valid
      (Transformation|AttachActorToComponent Component "None" "KeepWorld" "KeepWorld" "KeepWorld"))))

(fn HandleStuckPhysicsObject
  (params (Object :type Object%20Reference) (Component :type Primitive%20Component%20Object%20Reference))
  (bind _asactor (Utilities|Casting|CastToActor Object))
  (Utilities|IsValid (Actor|GetComponentbyClass _asactor "/Script/FortniteGame.FortPhysicsObjectComponent")
    (:Is%20Valid
      (Transformation|AttachActorToComponent Component))))

(fn HandleStuckProjectiles
  (params (Object :type Object%20Reference) (Component :type Primitive%20Component%20Object%20Reference))
  (Utilities|FlowControl|Sequence
    (:then_0
      (HandleSticks|HandleStuckBuilding Object Component))
    (:then_1
      (HandleSticks|HandleStuckVehicle Object Component))
    (:then_2
      (HandleSticks|HandleStuckSupplyDrop Object Component))))

(fn HandleStuckSupplyDrop
  (params (Object :type Object%20Reference) (Component :type Primitive%20Component%20Object%20Reference))
  (Utilities|FlowControl|Sequence
    (:then_0
      (Utilities|Casting|CastToAthenaSupplyDrop Object)
      (Transformation|AttachActorToComponent Component "None" "KeepWorld" "KeepWorld" "KeepWorld"))
    (:then_1
      (Utilities|Casting|CastToAthenaSupplyDropBalloon Object)
      (Transformation|AttachActorToComponent Component "None" "KeepWorld" "KeepWorld" "KeepWorld" false))))

(fn HandleStuckVehicle
  (params (Object :type Object%20Reference) (Component :type Primitive%20Component%20Object%20Reference))
  (Utilities|Casting|CastToFortAthenaVehicle Object)
  (Transformation|AttachActorToComponent Component "None" "KeepWorld" "KeepWorld" "KeepWorld" false))

(fn IncapacitateTarget
  (bind _target (Variables|Default|GetTarget))
  (bind _returnvalue (Pawn|GetController _target))
  (Components|Movement|StopMovementImmediately (Variables|/Script/Engine.Character|GetCharacterMovement (Variables|Default|GetTarget)))
  (Input|SetIgnoreMoveInput _returnvalue true)
  ;[1 1 1] Makes target face instigator
  (Pawn|SetControlRotation _returnvalue (Math|Rotator|FindLookatRotation (Transformation|GetActorLocation (Variables|Default|GetTarget)) (Transformation|GetActorLocation (Variables|Default|GetPlayerPawn)))))

(fn JackalCheck
  (bind _playerpawn (Variables|Default|GetPlayerPawn))
  (bind _returnvalue (Vehicle|GetVehicleInterface _playerpawn))
  (bind _returnvalue_2 (Utilities|ToObject(Interface) _returnvalue))
  (Utilities|IsValid _returnvalue_2
    (:Is%20Valid
      (Utilities|Casting|CastToFortAthenaJackalVehicle _returnvalue_2
        (:then
          (GameplayEffects|ApplyGameplayEffectToTarget (Ability|GetAbilitySystemComponent (Variables|Default|GetPlayerPawn)) "/Game/Athena/Items/Weapons/Abilities/HookGun/GE_Athena_HoverboardMontagePlaying.GE_Athena_HoverboardMontagePlaying_C" (Ability|GetAbilitySystemComponent (Variables|Default|GetPlayerPawn)) 1.0 (Utilities|Struct|MakeGameplayEffectContextHandle))
          (return))
        (:CastFailed
          (return))) ; TODO: Change this to a tag check rather than an explicit Cast event which will pull the Jackal into memory.
    )
    (:Is%20Not%20Valid
      (return))))

(fn MomentumCheck
  (params (A :type Vector) (A2 :type Vector))
  ;[1 1 1] Only add momentum if it's a net gain
  (return (select (> (Math|Vector|VectorLength (+ A2 (Variables|Default|GetMomentum))) (Math|Vector|VectorLength A)) (Variables|Default|GetMomentum) "0, 0, 0")))

(fn OnRep_HookStopped)

(fn OnRep_Momentum)

(fn OnRep_StruckActor)

(fn OnRep_Target)

(fn PickGrappleMontage
  (params (PitchAngle :type Float%20%28double-precision%29) (InHoverboard :type Boolean))
  (bind (_minpitch_6_c27dbdc64650f9ba5ebf09be2564561b _maxpitch_7_75538945434998adec50baa0edf1654a _sectiontoplay_12_2e20dbf74b43f4e8973e488482dca3f7) (Utilities|Struct|BreakToyAngleSelectorInfo (Utilities|Array|Get(aref) (Variables|Default|GetLaunchAngleVariations))))
  (LocalVariables|SetCleanedAngle (Math|Float|ClampAngle 0.0 -9e+01 9e+01))
  (for _array_element (Variables|Default|GetLaunchAngleVariations)
    (if (and (>= (LocalVariables|GetCleanedAngle) (Utilities|Struct|BreakToyAngleSelectorInfo _array_element)) (<= (LocalVariables|GetCleanedAngle) (.MaxPitch_7_75538945434998ADEC50BAA0EDF1654A (Utilities|Struct|BreakToyAngleSelectorInfo _array_element))))
      (return (.SectionToPlay_12_2E20DBF74B43F4E8973E488482DCA3F7 (Utilities|Struct|BreakToyAngleSelectorInfo _array_element)))))
  (return _sectiontoplay_12_2e20dbf74b43f4e8973e488482dca3f7))

(fn PickGrappleMontageSection
  (params (PitchAngle :type Float%20%28double-precision%29))
  (bind (_minpitch_6_c27dbdc64650f9ba5ebf09be2564561b _maxpitch_7_75538945434998adec50baa0edf1654a _sectiontoplay_12_2e20dbf74b43f4e8973e488482dca3f7) (Utilities|Struct|BreakToyAngleSelectorInfo (Utilities|Array|Get(aref) (Variables|Default|GetLaunchAngleVariations))))
  (LocalVariables|SetCleanedAngle (Math|Float|ClampAngle PitchAngle -9e+01 9e+01))
  (for _array_element (Variables|Default|GetLaunchAngleVariations)
    (if (and (>= (LocalVariables|GetCleanedAngle) (Utilities|Struct|BreakToyAngleSelectorInfo _array_element)) (<= (LocalVariables|GetCleanedAngle) (.MaxPitch_7_75538945434998ADEC50BAA0EDF1654A (Utilities|Struct|BreakToyAngleSelectorInfo _array_element))))
      (return (.SectionToPlay_12_2E20DBF74B43F4E8973E488482DCA3F7 (Utilities|Struct|BreakToyAngleSelectorInfo _array_element)))))
  (return _sectiontoplay_12_2e20dbf74b43f4e8973e488482dca3f7))

(fn PlayStickSounds
  (params (Location :type Vector) (HitActor :type Actor%20Object%20Reference))
  (Utilities|IsValid (Variables|Default|GetPlayerPawn)
    (:Is%20Valid
      (Utilities|IsValid HitActor
        (:Is%20Valid
          ;[1 1 1] Plays if an actor is hit
          (Utilities|FlowControl|Sequence
            (:then_0
              (if (Pawn|IsLocallyControlled (Variables|Default|GetPlayerPawn))
                (Audio|PlaySound2D (Variables|GrapplerSounds|GetSound_StuckToPlayer)) ; Plays on instigator
              ))
            (:then_1
              (if (Pawn|IsLocallyControlled (Utilities|Casting|CastToFortPlayerPawn HitActor))
                (Audio|PlaySound2D (Variables|GrapplerSounds|GetSound_StuckToPlayer)) ; Plays on target
                (else
                  _)))))
        (:Is%20Not%20Valid
          (if (Pawn|IsLocallyControlled (Variables|Default|GetPlayerPawn))
            ;[1 1 1] Plays if env is hit
            (Audio|PlaySound2D (Variables|GrapplerSounds|GetSound_Stick))
            (else
              _)))))
    (:Is%20Not%20Valid
      _))
  (Audio|PlaySoundatLocation (Variables|GrapplerSounds|GetSound_Stick) Location))

(fn SpawnRope
  (params (PlayerPawn :type Fort%20Player%20Pawn%20Object%20Reference))
  ;[1 1 1] Spawn The Rope
  (bind _returnvalue (Game|SpawnActorfromClass "/Game/Weapons/FORT_Crossbows/Blueprints/B_HookGunRope.B_HookGunRope_C" (Math|Transform|MakeTransform (Transformation|GetActorLocation)) "Undefined" "MultiplyWithRoot" (Variables|Default|GetWeapon%20Actor) PlayerPawn 0))
  (bind _output_get (Variables|Default|SetRope _returnvalue))
  (Class|FortHookGunRope|SetWeaponandProjectile _output_get (Variables|Default|GetWeapon%20Actor) self)
  (return _output_get))

(fn ConstructionScript
  (bind _speed (Variables|Default|GetSpeed))
  (bind _returnvalue (ScalableFloat|GetValueatLevel _speed))
  (bind _projectilemovementcomponent (Variables|Movement|GetProjectileMovementComponent))
  (|CallParentFunction)
  (Variables|/Script/Engine.ProjectileMovementComponent|SetInitialSpeed _returnvalue _projectilemovementcomponent)
  (Variables|/Script/Engine.ProjectileMovementComponent|SetMaxSpeed _returnvalue _projectilemovementcomponent)
  (Actor|SetLifeSpan (ScalableFloat|GetValueatLevel (Variables|Default|GetLifespan))))
